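# Physics asset for the SummerVivid Doll Female pickaxe, a two-handed melee cosmetic, defining collision and joints for its handle and dangling strap.
# One large capsule covers the handle bone, and two small capsules cover the dyn_strap bones.
# Constraints chain the second strap bone to the first and the first to the handle, with wider swing at the handle joint and tighter swing further along.
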PhysicsAsset /BRCosmetics/Weapons/FORT_Melee/Pickaxe_SummerVivid_Doll_Female/Meshes/SummerVivid_Doll_Female_Axe_Physics
bodies {
 Melee_TwoHanded_Handle [{shapeType=Capsule, length=72.8268, radius=10.8853, center=<0.7338 -0.7985 16.4119>, rotation=<rot 0.1806 -90 0.0572>}]
 dyn_strap_1 [{shapeType=Capsule, length=3.727, radius=2.5826, shapeName=dyn_strap_1_capsule, center=<3 0 0>, rotation=<rot 90 0 0>}]
 dyn_strap_2 [{shapeType=Capsule, length=2.9249, radius=2.2958, shapeName=dyn_strap_2_capsule, center=<3 0 0>, rotation=<rot 90 0 0>}]
}
constraints [
 {bone1Name=dyn_strap_2, bone2Name=dyn_strap_1, swing1LimitDegrees=20, swing2LimitDegrees=20, twistLimitDegrees=10}
 {bone1Name=dyn_strap_1, bone2Name=Melee_TwoHanded_Handle, swing1LimitDegrees=50, swing2LimitDegrees=50, twistLimitDegrees=10}
]
