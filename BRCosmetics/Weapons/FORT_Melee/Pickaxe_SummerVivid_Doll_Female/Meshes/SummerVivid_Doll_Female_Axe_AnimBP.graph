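# Animation Blueprint for the female SummerVivid Doll pickaxe that adds secondary physics motion to the axe mesh.
# The AnimGraph runs the input pose through a RigidBody node using the axe's physics asset, simulated relative to b_AR_Root with overridden gravity.
# EventGraph logic for detecting the offhand mesh by name and setting four "Hide Element" material parameters is present but not wired to either event.
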AnimBlueprint /BRCosmetics/Weapons/FORT_Melee/Pickaxe_SummerVivid_Doll_Female/Meshes/SummerVivid_Doll_Female_Axe_AnimBP
target_skeleton=&/SummerVivid_Doll_Female_Axe_Skeleton
graphs [
 AnimGraph {nodes [
   AnimGraphNode_SubInput_0 = Input Pose()
   n1 = RigidBody(ComponentPose=Local To Component(LocalPose=@AnimGraphNode_SubInput_0)) {
    bClampLinearTranslationLimitToRefPose=true
    bOverrideWorldGravity=true
    overridePhysicsAsset=&/SummerVivid_Doll_Female_Axe_Physics
    simulationSpace=BaseBoneSpace
    alphaBoolBlend {blendOption=Linear}
    baseBoneRef {boneName=b_AR_Root}
    componentLinearAccScale=<0.7 0.7 0.7>
    componentLinearVelScale=<0.7 0.7 0.7>
    overrideWorldGravity {z=-980}
   }
   Output Pose(Result=Component To Local(ComponentPose=@n1)) {layerGroup=None}
  ]}
 EventGraph {nodes [
   Branch(Condition=NOT Boolean(A=Get offset_weapon() -> (offset_weapon:bool)))
   > n1 = Sequence!(execute) -> (then_0, then_1, then_2:Exec, then_3:Exec)
   > n2 = Branch()
   n3 = GetOwningComponent()
   > n4 = Get Dynamic Material Instance(SkeletalMeshComponent=@n3)
   > SetScalarParameterValue(self=@n4, ParameterName=Hide Element 01, Value=1)
   n5 = Get Dynamic Material Instance(execute=@n2.else, SkeletalMeshComponent=@n3)
   > SetScalarParameterValue(self=@n5, ParameterName=Hide Element 01, Value=0)
   n6 = Branch(execute=@n1.then_1)
   n7 = GetOwningComponent()
   > n8 = Get Dynamic Material Instance(SkeletalMeshComponent=@n7)
   > SetScalarParameterValue(self=@n8, ParameterName=Hide Element 02, Value=1)
   n9 = Get Dynamic Material Instance(execute=@n6.else, SkeletalMeshComponent=@n7)
   > SetScalarParameterValue(self=@n9, ParameterName=Hide Element 02, Value=0)
   n10 = Branch(execute=@n1.then_2)
   n11 = GetOwningComponent()
   > n12 = Get Dynamic Material Instance(SkeletalMeshComponent=@n11)
   > SetScalarParameterValue(self=@n12, ParameterName=Hide Element 03, Value=1)
   n13 = Get Dynamic Material Instance(execute=@n10.else, SkeletalMeshComponent=@n11)
   > SetScalarParameterValue(self=@n13, ParameterName=Hide Element 03, Value=0)
   n14 = Branch(execute=@n1.then_3)
   n15 = GetOwningComponent()
   > n16 = Get Dynamic Material Instance(SkeletalMeshComponent=@n15)
   > SetScalarParameterValue(self=@n16, ParameterName=Hide Element 04, Value=1)
   n17 = Get Dynamic Material Instance(execute=@n14.else, SkeletalMeshComponent=@n15)
   > SetScalarParameterValue(self=@n17, ParameterName=Hide Element 04, Value=0)
   n18 = Set Mesh_Name(Mesh_Name:Skeletal Mesh Component Object Reference=GetPartSkeletalMeshComponent()) -> (Output_Get:Skeletal Mesh Component Object Reference)
   > n19 = Branch(Condition="Equal Exactly (String)"(A="To String (Object)"!(InObj:Object Reference=@n18) -> (ReturnValue:String), B=WeaponMeshOffhand))
   > Set offset_weapon(offset_weapon:bool=true) -> (Output_Get:bool)
   Set offset_weapon(execute=@n19.else, offset_weapon=false)
   TryGetPawnOwner()
   Event BlueprintUpdateAnimation()
   Event BlueprintInitializeAnimation()
  ]}
]
variables [
 Mesh_Name {type=SkeletalMeshComponent}
 offset_weapon {type=boolean}
]
